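# Ambient SoundCue for a wooden boat on water, layering intermittent wood impacts over a quieter, pitched-down looping pontoon bed.
# Two looping branches each wait a random few seconds, then play one of three impact samples with slight pitch variation, mixed with the bed.
# Overridden capsule attenuation with listener focus and a two-voice concurrency limit shape playback, which also sends to the exterior ambient bus.
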SoundCue /Game/Sounds/Helios/Environment_Actors/WoodBoat/Spot_Wood_Boat_Impacts
root=Attenuation_1
nodes [
 {id=Attenuation_1, children [Mixer_0]}
 {id=Delay_0, children [Modulator_0], properties {delayMax=6, delayMin=3}}
 {id=Delay_1, children [Modulator_2], properties {delayMax=5, delayMin=2}}
 {id=Looping_0, children [Delay_0]}
 {id=Looping_1, children [Delay_1]}
 {id=Mixer_0, children [Looping_0,Looping_1,Modulator_1], properties {inputVolume [0.6,0.6,0.3]}}
 {id=Modulator_0, children [Random_0], properties {pitchMax=1.1, pitchMin=0.9}}
 {id=Modulator_1, children [WavePlayer_2], properties {pitchMax=0.8, pitchMin=0.8}}
 {id=Modulator_2, children [Random_2], properties {pitchMax=1.1, pitchMin=0.9}}
 {id=Random_0, children [WavePlayer_6,WavePlayer_7,WavePlayer_8], properties {weights [1,1,1]}}
 {id=Random_2, children [WavePlayer_0,WavePlayer_1,WavePlayer_9], properties {weights [1,1,1]}}
 {id=WavePlayer_0, properties {soundWaveAssetPtr=/Game/Athena/Sounds/Water/Ambient/Spot_Water_Pontoon_Wood_Impact_01}}
 {id=WavePlayer_1, properties {soundWaveAssetPtr=/Game/Athena/Sounds/Water/Ambient/Spot_Water_Pontoon_Wood_Impact_02}}
 {id=WavePlayer_2, properties {bLooping=true, soundWaveAssetPtr=/Game/Athena/Sounds/Water/Ambient/Amb_Water_Wood_Pontoon_Loop}}
 {id=WavePlayer_6, properties {soundWaveAssetPtr=/Game/Athena/Sounds/Water/Ambient/Spot_Water_Pontoon_Wood_Impact_04}}
 {id=WavePlayer_7, properties {soundWaveAssetPtr=/Game/Athena/Sounds/Water/Ambient/Spot_Water_Pontoon_Wood_Impact_05}}
 {id=WavePlayer_8, properties {soundWaveAssetPtr=/Game/Athena/Sounds/Water/Ambient/Spot_Water_Pontoon_Wood_Impact_06}}
 {id=WavePlayer_9, properties {soundWaveAssetPtr=/Game/Athena/Sounds/Water/Ambient/Spot_Water_Pontoon_Wood_Impact_03}}
]
properties {
 bOverrideAttenuation=true
 bOverrideConcurrency=true
 maxDistance=2097152
 soundClassObject=/Game/Sounds/FortSoundClassesAndModes/Fort_Water_Ambient
 volumeMultiplier=1.5
 attenuationOverrides {
  attenuationShape=Capsule
  bEnableFocusInterpolation=true
  bEnableListenerFocus=true
  falloffDistance=750
  lPFFrequencyAtMax=500
  nonFocusVolumeAttenuation=0.5
  attenuationShapeExtents=<590 500>
 }
 busSends [{audioBus=null, minSendLevel=0.6, sendLevel=1, soundSourceBus=&AmbientAudio_Exterior_SourceBus}]
 concurrencyOverrides {maxCount=2}
}
